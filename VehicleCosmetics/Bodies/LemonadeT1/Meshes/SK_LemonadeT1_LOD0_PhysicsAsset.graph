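# Physics asset for the LemonadeT1 vehicle body cosmetic, giving its LOD0 skeletal mesh a simple collision shape.
# It holds a single body on the frame bone, made of one large capsule that is offset and rotated about a quarter turn on two axes.
# With no other bodies and no constraints, the vehicle body acts as one piece with no articulated parts.
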PhysicsAsset /VehicleCosmetics/Bodies/LemonadeT1/Meshes/SK_LemonadeT1_LOD0_PhysicsAsset
bodies {frame [{shapeType=Capsule, length=229.0066, radius=109.4911, center=<18.4461 -80.8871 -0.3328>, rotation=<rot 0 -88.9484 -89.864>}]}
constraints []
